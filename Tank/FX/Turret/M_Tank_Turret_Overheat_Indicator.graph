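# Translucent Default Lit material for Niagara mesh particles, apparently the tank turret's overheat indicator.
# A UV gradient, with optional alternate UV channel and flip, slides by a particle dynamic offset.
# It masks an inverted square-outline texture tinted by particle color, added to a faint base and fed through an emissive adjustment function.
# A separate chain multiplies the gradient by depth fade and optional, invertible Fresnel, also driven by dynamic parameters.
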Material /Tank/FX/Turret/M_Tank_Turret_Overheat_Indicator
graph {
 properties {bUsedWithNiagaraMeshParticles=true, blendMode=BLEND_Translucent, pixelDepthOffsetMode=PDOM_Legacy, usedShadingModels=Default Lit}
 nodes [
  n1 = OneMinus(None=TextureSample(UVs=TextureCoordinate()) {texture=/Game/Effects/Fort_Effects/Emotes/DontBeASquare/T_SquareOutline_alpha}.RGB)
  n2 = ParticleColor()
  n3 = Multiply(A=Multiply(A=@n1) {constB=0.05}, B=@n2.RGB) {constB=0.1}
  n4 = MaterialFunctionCall(XY=TextureCoordinate()) {materialFunction=&Swizzle}
  n5 = DynamicParameter() {paramNames [UVOffsetY, DepthFadeDistance, FresnelExp, InvertFresnel]}
  n6 = TextureCoordinate() {coordinateIndex=1}
  n7 = LinearInterpolate(
   A=Add(A=ComponentMask(None=@n4.YX) {r=true}, B=@n5.UVOffsetY)
   B=Add(A=ComponentMask(None=@n6) {g=true}, B=@n5.UVOffsetY)
   Alpha=ScalarParameter() {parameterName=UseAltUVMap}
  )
  n8 = Clamp(None=LinearInterpolate(A=@n7, B=OneMinus(None=@n7), Alpha=ScalarParameter() {parameterName=FlipUV}))
  n9 = Multiply(A=@n2.RGB, B=Multiply(A=Clamp(None=@n1), B=@n8))
  MaterialFunctionCall(Emissive=Clamp(None=Add(A=@n3, B=@n9) {constB=-0.5})) {materialFunction=&MF_FX_EmissiveAdjustments}
  n10 = DepthFade(FadeDistance=Multiply(A=ScalarParameter() {defaultValue=1, parameterName=DepthFadeDistance}, B=@n5.DepthFadeDistance))
  n11 = Fresnel(ExponentIn=@n5.FresnelExp)
  n12 = LinearInterpolate(A=@n11, B=OneMinus(None=@n11), Alpha=@n5.InvertFresnel)
  Multiply(
   A=Add(A=Multiply() {constB=0}, B=@n8) {constB=-0.5}
   B=LinearInterpolate(A=@n10, B=Multiply(A=@n10, B=@n12), Alpha=ScalarParameter() {defaultValue=1, parameterName=UseFresnel})
  )
  Multiply()
  ScalarParameter() {defaultValue=1, parameterName=FresnelExp}
  ScalarParameter() {defaultValue=1, parameterName=InvertFresnel}
 ]
}
